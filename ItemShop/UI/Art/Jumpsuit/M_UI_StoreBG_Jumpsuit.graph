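# UI-domain material that paints a grungy, spray-painted background for the Jumpsuit item shop.
# Summed spray and overspray masks, each placed by position and scale parameters, darken a gray grunge base built on fill-scaled UVs.
# A bomb-spray stencil adds a purple radial gradient worn by the grunge, then a parameterized vertical falloff darkens top and bottom edges.
# An arrow layer and two duplicate grunge chains do not feed that composite.
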Material /ItemShop/UI/Art/Jumpsuit/M_UI_StoreBG_Jumpsuit
graph {
 properties {bUsedWithStaticMesh=false, materialDomain=MD_UI, pixelDepthOffsetMode=PDOM_Legacy, usedShadingModels=Default Lit}
 nodes [
  Constant3Vector() {constant=<rgba 0.0208 0.0208 0.0208 1>}
  Constant3Vector() {constant=<rgba 0.0542 0.9473 0 1>}
  n1 = Constant3Vector() {constant=<rgba 0.0469 0.0469 0.0469 1>}
  n2 = MaterialFunctionCall() {materialFunction=&MF_UI_UVScaling}
  n3 = Constant() {r=0.896}
  n4 = TextureSample(UVs=MaterialFunctionCall!(
    Texture Scale=VectorParameter() {parameterName=SprayBloom5Scale, defaultValue=<rgba 0.524 0.5 0 1>}.RGB
    UVs=MaterialFunctionCall(UVs=@n2.Scaled To Fill, Texture Scale=@n3) {materialFunction=&ScaleUVsByCenter}.UVs
    Offset XY Texturesize Z=VectorParameter() {parameterName=SprayBloom5 Position, defaultValue=<rgba -0.592 1.2492 1 1>}.RGB
   ) {materialFunction=/Engine/Functions/Engine_MaterialFunctions01/Texturing/ScaleUVsAroundPoint}.UVs) {samplerType=SAMPLERTYPE_Normal, texture=&/T_UI_Spray}
  n5 = VectorParameter() {parameterName=SprayBloom4 Scale, defaultValue=<rgba 0.65 0.618 0 1>}
  n6 = MaterialFunctionCall(UVs=@n2.Scaled To Fill, Texture Scale=Constant() {r=0.5}) {materialFunction=&ScaleUVsByCenter}
  n7 = VectorParameter() {parameterName=SprayBloom4 Position, defaultValue=<rgba -2.465 -0.184 1 1>}
  n8 = Saturate(None=TextureSample(UVs=MaterialFunctionCall(Texture Scale=@n5.RGB, UVs=@n6.UVs, Offset XY Texturesize Z=@n7.RGB) {materialFunction=&ScaleUVsAroundPoint}.UVs) {samplerType=SAMPLERTYPE_Normal, texture=&T_UI_Spray}.R)
  n9 = Constant() {r=0.35}
  n10 = TextureSample(UVs=MaterialFunctionCall(
    Texture Scale=VectorParameter() {parameterName=SprayBloom2 Scale, defaultValue=<rgba 1.0662 1.1436 1 1>}.RGB
    UVs=MaterialFunctionCall(UVs=@n2.Scaled To Fill, Texture Scale=@n9) {materialFunction=&ScaleUVsByCenter}.UVs
    Offset XY Texturesize Z=VectorParameter() {parameterName=SprayBloom2 Position, defaultValue=<rgba -21 -2 1 0>}.RGB
   ) {materialFunction=&ScaleUVsAroundPoint}.UVs) {samplerType=SAMPLERTYPE_Normal, texture=&T_UI_Spray}
  n11 = Constant() {r=0.5}
  n12 = TextureSample(UVs=MaterialFunctionCall(
    Texture Scale=VectorParameter() {parameterName=SprayBlob3 Scale, defaultValue=<rgba 0.5 0.5 1 1>}.RGB
    UVs=MaterialFunctionCall(UVs=@n2.Scaled To Fill, Texture Scale=@n11) {materialFunction=&ScaleUVsByCenter}.UVs
    Offset XY Texturesize Z=VectorParameter() {parameterName=SprayBlob3 Position, defaultValue=<rgba 0.648 1.3395 1 1>}.RGB
   ) {materialFunction=&ScaleUVsAroundPoint}.UVs) {samplerType=SAMPLERTYPE_Normal, texture=&T_UI_Spray}
  n13 = Add(A=Add(A=Saturate(None=@n4.R), B=Multiply(A=@n8) {constB=0.8}), B=Add(A=Saturate(None=@n10.G), B=Saturate(None=@n12.R)))
  n14 = MaterialFunctionCall(
   Texture Scale=VectorParameter() {parameterName=Overspray2 Scale, defaultValue=<rgba 0.508 0.5 1 0>}.RGB
   UVs=MaterialFunctionCall(UVs=@n2.Scaled To Fill, Texture Scale=Constant() {r=0.75}) {materialFunction=&ScaleUVsByCenter}.UVs
   Offset XY Texturesize Z=VectorParameter() {parameterName=Overspray2 Position, defaultValue=<rgba 1.0386 0.1 1 1>}.RGB
  ) {materialFunction=&ScaleUVsAroundPoint}
  n15 = MaterialFunctionCall(
   Texture Scale=VectorParameter() {parameterName=Overspray1 Scale, defaultValue=<rgba 0.5 0.5 1 1>}.RGB
   UVs=MaterialFunctionCall(UVs=@n2.Scaled To Fill, Texture Scale=Constant() {r=0.75}) {materialFunction=&ScaleUVsByCenter}.UVs
   Offset XY Texturesize Z=VectorParameter() {parameterName=Overspray1 Position, defaultValue=<rgba -0.245 0.27 0.456 1>}.RGB
  ) {materialFunction=&ScaleUVsAroundPoint}
  n16 = Add(
   A=Saturate(None=TextureSample(UVs=@n14.UVs) {samplerType=SAMPLERTYPE_Normal, texture=&/T_UI_OverSpray}.G)
   B=Saturate(None=TextureSample(UVs=@n15.UVs) {samplerType=SAMPLERTYPE_Normal, texture=&T_UI_OverSpray}.R)
  )
  n17 = VectorParameter() {parameterName=Spray2 Scale, defaultValue=<rgba 0.932 0.9797 1.3224 1>}
  n18 = MaterialFunctionCall(UVs=@n2.Scaled To Fill, Texture Scale=Constant() {r=0.6}) {materialFunction=&ScaleUVsByCenter}
  n19 = VectorParameter() {parameterName=Spray2 Position, defaultValue=<rgba 13.306 -16.9033 1 1>}
  n20 = Saturate(None=TextureSample(UVs=MaterialFunctionCall(Texture Scale=@n17.RGB, UVs=@n18.UVs, Offset XY Texturesize Z=@n19.RGB) {materialFunction=&ScaleUVsAroundPoint}.UVs) {samplerType=SAMPLERTYPE_Normal, texture=&T_UI_Spray}.G)
  n21 = MaterialFunctionCall(
   Texture Scale=VectorParameter() {parameterName=Spray1 Scale, defaultValue=<rgba 0.532 0.596 1 1>}.RGB
   UVs=MaterialFunctionCall(UVs=@n2.Scaled To Fill, Texture Scale=Constant() {r=0.5}) {materialFunction=&ScaleUVsByCenter}.UVs
   Offset XY Texturesize Z=VectorParameter() {parameterName=Spray1 Position, defaultValue=<rgba -1.4326 -0.624 1 1>}.RGB
  ) {materialFunction=&ScaleUVsAroundPoint}
  n22 = Multiply(A=Saturate(None=TextureSample(UVs=@n21.UVs) {samplerType=SAMPLERTYPE_Normal, texture=&T_UI_Spray}.R)) {constB=0.8}
  n23 = Multiply(A=Add(A=Add(A=@n13, B=@n16), B=Add(A=@n20, B=@n22))) {constB=1.25}
  n24 = MaterialFunctionCall(
   UVs=MaterialFunctionCall() {materialFunction=&MF_UI_UVScaling}.Scaled To Fill
   Texture Scale=Constant() {r=0.6014}
  ) {materialFunction=&ScaleUVsByCenter}
  n25 = TextureSample(UVs=@n24.UVs) {samplerType=SAMPLERTYPE_Normal, texture=&/T_UI_Grunge1}
  n26 = Add(A=Multiply(A=@n25.R) {constB=0.5}) {constB=0.5}
  n27 = Constant3Vector() {constant=<rgba 0.1095 0.1046 0.117 1>}
  n28 = Saturate(None=@n25.G)
  n29 = LinearInterpolate(
   A=LinearInterpolate(A=VectorParameter() {defaultValue=<rgba 0.1845 0.1845 0.1845 1>}.RGB, B=Multiply(A=@n26, B=@n27), Alpha=@n26)
   B=Multiply(A=@n25.G, B=Constant3Vector() {constant=<rgba 0.026 0.026 0.026 1>})
   Alpha=@n28
  )
  n30 = Constant3Vector() {constant=<rgba 0.2502 0.0021 0.6308 1>}
  n31 = Constant3Vector() {constant=<rgba 0.2648 0.0263 0.6308 1>}
  n32 = Constant() {r=0.8088}
  n33 = Constant() {r=0.9636}
  n34 = MaterialFunctionCall() {materialFunction=&MF_UI_UVScaling}
  n35 = MaterialFunctionCall(UV=Multiply(
    A=MaterialFunctionCall(X=@n32, Y=@n33) {materialFunction=&MakeFloat2}.Result
    B=@n34.Scaled To Fill
   )) {materialFunction=&MF_UI_RadialGradient}
  n36 = Constant() {r=0.17}
  n37 = TextureSample(UVs=MaterialFunctionCall(
    Texture Scale=VectorParameter() {parameterName=Bomb Scale, defaultValue=<rgba 2.5156 2.5535 1.3224 1>}.RGB
    UVs=MaterialFunctionCall(UVs=@n34.Scaled To Fill, Texture Scale=@n36) {materialFunction=&ScaleUVsByCenter}.UVs
    Offset XY Texturesize Z=VectorParameter() {parameterName=Bomb Position, defaultValue=<rgba 0.2 0.12 0.136 0>}.RGB
   ) {materialFunction=&ScaleUVsAroundPoint}.UVs) {samplerType=SAMPLERTYPE_Alpha, texture=&/T_UI_BombSpray}
  n38 = LinearInterpolate(
   A=@n29
   B=LinearInterpolate(A=@n29, B=LinearInterpolate(A=@n30, B=@n31, Alpha=@n35.Smooth), Alpha=Multiply(B=@n37.R) {constA=0.65})
   Alpha=@n37.R
  )
  n39 = Multiply(A=@n28) {constB=0.45}
  n40 = Abs(None=ConstantBiasScale(None=TextureCoordinate()) {bias=-0.5, scale=2})
  n41 = Multiply(
   A=Subtract(A=ComponentMask(None=@n40) {g=true}, B=ScalarParameter() {defaultValue=0.7349, parameterName=Subtract})
   B=ScalarParameter() {defaultValue=1.4621, parameterName=Falloff}
  )
  LinearInterpolate(
   A=LinearInterpolate(A=LinearInterpolate(A=@n29, B=@n1, Alpha=@n23), B=LinearInterpolate(A=@n38, B=@n29, Alpha=@n39), Alpha=@n37.R)
   B=Constant3Vector() {constant=<rgba 0.0049 0.0052 0.0042 1>}
   Alpha=Multiply(A=Saturate(None=@n41), B=ScalarParameter() {defaultValue=2.5483, parameterName=Intensity})
  )
  n42 = MaterialFunctionCall(Texture Scale=Constant() {r=0.7}) {materialFunction=&ScaleUVsByCenter}
  n43 = Constant() {r=0.5}
  n44 = TextureSample(UVs=MaterialFunctionCall(
    Texture Scale=VectorParameter() {parameterName=Arrow Scale, defaultValue=<rgba 0.412 0.5 1.3224 1>}.RGB
    UVs=MaterialFunctionCall(UVs=@n42.UVs, "Rotation Angle (0-1)"=@n43) {materialFunction=&CustomRotator}.Rotated Values
    Offset XY Texturesize Z=VectorParameter() {parameterName=Arrow Position, defaultValue=<rgba -0.6228 -0.1824 0.488 1>}.RGB
   ) {materialFunction=&ScaleUVsAroundPoint}.UVs) {texture=&T_UI_BombSpray}
  LinearInterpolate(B=Multiply(A=Multiply(A=@n44.G) {constB=0.4}, B=Constant3Vector() {constant=<rgba 0.4621 0.0513 0.9047 1>}), Alpha=@n44.G)
  n45 = MaterialFunctionCall(
   UVs=MaterialFunctionCall() {materialFunction=&MF_UI_UVScaling}.Scaled To Fill
   Texture Scale=Constant() {r=0.6014}
  ) {materialFunction=&ScaleUVsByCenter}
  n46 = TextureSample(UVs=@n45.UVs) {samplerType=SAMPLERTYPE_Normal, texture=&T_UI_Grunge1}
  n47 = Add(A=Multiply(A=@n46.R) {constB=0.5}) {constB=0.5}
  n48 = Constant3Vector() {constant=<rgba 0.1094 0.1094 0.1094 1>}
  LinearInterpolate(
   A=LinearInterpolate(A=VectorParameter() {defaultValue=<rgba 0.1845 0.1845 0.1845 1>}.RGB, B=Multiply(A=@n47, B=@n48), Alpha=@n47)
   B=Multiply(A=@n46.G, B=Constant3Vector() {constant=<rgba 0.01 0.01 0.01 1>})
   Alpha=Saturate(None=@n46.G)
  )
  n49 = MaterialFunctionCall(
   UVs=MaterialFunctionCall() {materialFunction=&MF_UI_UVScaling}.Scaled To Fill
   Texture Scale=Constant() {r=0.6014}
  ) {materialFunction=&ScaleUVsByCenter}
  n50 = TextureSample(UVs=@n49.UVs) {samplerType=SAMPLERTYPE_Normal, texture=&T_UI_Grunge1}
  n51 = Add(A=Multiply(A=@n50.R) {constB=0.5}) {constB=0.5}
  n52 = Constant3Vector() {constant=<rgba 0.1146 0.1146 0.1146 1>}
  LinearInterpolate(
   A=LinearInterpolate(A=VectorParameter() {defaultValue=<rgba 0.1845 0.1845 0.1845 1>}.RGB, B=Multiply(A=@n51, B=@n52), Alpha=@n51)
   B=Multiply(A=@n50.G, B=Constant3Vector() {constant=<rgba 0.01 0.01 0.01 1>})
   Alpha=Saturate(None=@n50.G)
  )
  MaterialFunctionCall(Rotation Center=Constant() {r=0.75}, "Rotation Angle (0-1)"=Constant() {r=0.05}) {materialFunction=&CustomRotator}
  OneMinus(None=ConstantBiasScale(None=@n24.UVs))
  OneMinus(None=ConstantBiasScale(None=@n45.UVs))
  OneMinus(None=ConstantBiasScale(None=@n49.UVs))
  VectorParameter() {defaultValue=<rgba 0.1845 0.1845 0.1845 1>}
 ]
}
